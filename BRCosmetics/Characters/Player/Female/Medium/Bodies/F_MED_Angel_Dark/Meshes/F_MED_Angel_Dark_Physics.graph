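# Physics asset for the F_MED_Angel_Dark female player body cosmetic, covering a dangling ring and zipper plus the upper spine.
# Capsules on spine_04 and spine_05 form torso bodies, with the zipper chain constrained to spine_04 and the ring chain to spine_05.
# Each chain links a small master capsule with free angular motion to an end body (box ring, capsule zipper) with angular limits and tight twist.
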PhysicsAsset /BRCosmetics/Characters/Player/Female/Medium/Bodies/F_MED_Angel_Dark/Meshes/F_MED_Angel_Dark_Physics
bodies {
 dyn_ring_01 [{
   shapeType=Box
   extentX=2.2893
   extentY=5.0688
   extentZ=6.4361
   shapeName=dyn_ring_01_box
   center=<2.8799 0.0001 0.0004>
   rotation=<rot -0.0075 -90 -0.0201>
  }]
 dyn_ring_MSTR [{shapeType=Capsule, length=0.505, radius=0.505, shapeName=dyn_ring_MSTR_capsule, rotation=<rot 0 -90 0>}]
 dyn_zipper_01 [{shapeType=Capsule, length=1.505, radius=1.3175, shapeName=dyn_zipper_01_capsule, center=<2.9483 0.6491 0>, rotation=<rot 90 0 0>}]
 dyn_zipper_MSTR [{shapeType=Capsule, length=0.505, radius=0.505, shapeName=dyn_zipper_MSTR_capsule, rotation=<rot 0 -90 0>}]
 spine_04 [{shapeType=Capsule, length=14.3437, radius=13.6226}]
 spine_05 [{shapeType=Capsule, length=18.5639, radius=8.3747, center=<1.8644 0.0001 0>}]
}
constraints [
 {bone1Name=spine_05, bone2Name=spine_04, swing1LimitDegrees=45, swing2LimitDegrees=45, twistLimitDegrees=45}
 {
  bone1Name=dyn_zipper_MSTR
  bone2Name=spine_04
  swing1LimitDegrees=45
  swing1Motion=Free
  swing2LimitDegrees=45
  swing2Motion=Free
  twistLimitDegrees=45
  twistMotion=Free
 }
 {bone1Name=dyn_zipper_01, bone2Name=dyn_zipper_MSTR, swing1LimitDegrees=15, swing2LimitDegrees=15, twistLimitDegrees=5}
 {
  bone1Name=dyn_ring_MSTR
  bone2Name=spine_05
  swing1LimitDegrees=45
  swing1Motion=Free
  swing2LimitDegrees=45
  swing2Motion=Free
  twistLimitDegrees=45
  twistMotion=Free
 }
 {bone1Name=dyn_ring_01, bone2Name=dyn_ring_MSTR, swing1LimitDegrees=45, swing2LimitDegrees=45, twistLimitDegrees=5}
]
