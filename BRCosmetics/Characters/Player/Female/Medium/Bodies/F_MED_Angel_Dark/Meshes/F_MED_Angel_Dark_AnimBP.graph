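# Post-copy AnimBlueprint for the F_MED_Angel_Dark female body mesh, adding deformation and secondary motion to its ring and zipper bones.
# It copies the pose from the attached parent mesh, runs the deform Control Rig, then applies weighted multi-bone constraints to dyn_ring_MSTR and dyn_zipper_MSTR.
# A RigidBody node simulates the override physics asset relative to spine_03 with damped component motion.
# A three-target zipper constraint is left unconnected, and the event graph has no logic.
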AnimBlueprint /BRCosmetics/Characters/Player/Female/Medium/Bodies/F_MED_Angel_Dark/Meshes/F_MED_Angel_Dark_AnimBP
target_skeleton=&/F_MED_Angel_Dark_Skeleton
graphs [
 AnimGraph {nodes [
   "Constraint - dyn_zipper_MSTR (PPP)"!(ComponentPose, Alpha=1, ConstraintWeights_0=0.0006, ConstraintWeights_1:float=0.5084, ConstraintWeights_2:float=0.491) -> (Pose) {
    alphaBoolBlend {blendOption=Linear}
    boneToModify {boneName=dyn_zipper_MSTR}
    constraintSetup [{targetBone {boneName=deform_pectoral_fr_l}}, {targetBone {boneName=spine_04}}, {targetBone {boneName=spine_05}}]
    constraintWeights [1,1,1]
   }
   n1 = Control Rig(Source=Copy Pose From Mesh() {bUseAttachedParent=true}) {
    alphaBoolBlend {blendOption=Linear}
    controlRigAssetReference {blueprintRigClass=&/F_MED_Angel_Dark_Deform_CtrlRig.F_MED_Angel_Dark_Deform_CtrlRig_C}
   }
   n2 = "Constraint - dyn_ring_MSTR (PPPPPP)"!(
    ComponentPose=Local To Component(LocalPose=@n1)
    Alpha=1
    ConstraintWeights_0=0.0528
    ConstraintWeights_1:float=0.0169
    ConstraintWeights_2:float=0.6077
    ConstraintWeights_3:float=0.0004
    ConstraintWeights_4:float=0.1753
    ConstraintWeights_5:float=0.1469
   ) -> (Pose) {
    alphaBoolBlend {blendOption=Linear}
    boneToModify {boneName=dyn_ring_MSTR}
    constraintSetup [
     {targetBone {boneName=head}}
     {targetBone {boneName=clavicle_l}}
     {targetBone {boneName=spine_05}}
     {targetBone {boneName=clavicle_r}}
     {targetBone {boneName=neck_02}}
     {targetBone {boneName=neck_01}}
    ]
    constraintWeights [1,1,1,1,1,1]
   }
   n3 = "Constraint - dyn_zipper_MSTR (PPPPPP)"!(
    ComponentPose=@n2
    Alpha=1
    ConstraintWeights_0=0.1058
    ConstraintWeights_1:float=0
    ConstraintWeights_2:float=0.0004
    ConstraintWeights_3:float=0.3701
    ConstraintWeights_4:float=0.5237
    ConstraintWeights_5:float=0
   ) -> (Pose) {
    alphaBoolBlend {blendOption=Linear}
    boneToModify {boneName=dyn_zipper_MSTR}
    constraintSetup [
     {targetBone {boneName=clavicle_l}}
     {targetBone {boneName=deform_deltoid_fr_l}}
     {targetBone {boneName=deform_pectoral_fr_l}}
     {targetBone {boneName=spine_04}}
     {targetBone {boneName=spine_05}}
     {targetBone {boneName=upperarm_twist_01_l}}
    ]
    constraintWeights [1,1,1,1,1,1]
   }
   n4 = RigidBody(ComponentPose=@n3) {
    bClampLinearTranslationLimitToRefPose=true
    bOverrideWorldGravity=true
    overridePhysicsAsset=&/F_MED_Angel_Dark_Physics
    simulationSpace=BaseBoneSpace
    alphaBoolBlend {blendOption=Linear}
    baseBoneRef {boneName=spine_03}
    componentAppliedLinearAccClamp=<3000 3000 3000>
    componentLinearAccScale=<0.3 0.3 0.5>
    componentLinearVelScale=<0.1 0.1 0.2>
    overrideWorldGravity {z=-980}
   }
   Output Pose(Result=Component To Local(ComponentPose=@n4)) {layerGroup=None}
  ]}
 EventGraph {nodes [Event BlueprintUpdateAnimation()]}
]
variables []
